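; Blueprint for the female Dino Hunter pickaxe cosmetic, which makes its idle particle effect react to melee swings.
; Left and right swing events call the parent, set the idle VFX component's SwingSpeed float parameter to 1, then reset it to 0 after a half-second delay.
; Both swing-end events also reset SwingSpeed to 0.
; All four swing handlers print debug strings, and BeginPlay and Tick are empty.
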(event EventBeginPlay)

(event EventTick
  (params (DeltaSeconds :type Float%20%28single-precision%29)))

(event EventSwingRight
  (bind _returnvalue (CallFunction|GetIdleVFXComponent))
  (|CallParentFunction)
  (Development|PrintString "RightSwing")
  (Effects|Components|ParticleSystem|SetFloatParameter _returnvalue "SwingSpeed" 1.0)
  (Utilities|FlowControl|Delay 0.5)
  (Effects|Components|ParticleSystem|SetFloatParameter _returnvalue "SwingSpeed" 0.0))

(event EventSwingRightEnd
  (|CallParentFunction)
  (Development|PrintString "RightSwingEnd")
  (Effects|Components|ParticleSystem|SetFloatParameter (CallFunction|GetIdleVFXComponent) "SwingSpeed" 0.0))

(event EventSwingLeft
  (bind _returnvalue (CallFunction|GetIdleVFXComponent))
  (|CallParentFunction)
  (Development|PrintString "LeftSwing")
  (Effects|Components|ParticleSystem|SetFloatParameter _returnvalue "SwingSpeed" 1.0)
  (Utilities|FlowControl|Delay 0.5)
  (Effects|Components|ParticleSystem|SetFloatParameter _returnvalue "SwingSpeed" 0.0))

(event EventSwingLeftEnd
  (|CallParentFunction)
  (Development|PrintString "LeftSwingEnd")
  (Effects|Components|ParticleSystem|SetFloatParameter (CallFunction|GetIdleVFXComponent) "SwingSpeed" 0.0))

(fn ConstructionScript
  (|CallParentFunction))
