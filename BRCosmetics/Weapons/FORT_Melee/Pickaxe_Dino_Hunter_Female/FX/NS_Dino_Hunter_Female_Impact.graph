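# Niagara impact effect for the female Dino Hunter pickaxe: a brief flash with a burst of sparks.
# Spark_Impact spawns one short-lived sprite that swells then shrinks, while Sparks sprays randomized elongated sprites in a cone, shaped by drag, gravity and curl noise.
# User.SmokeColorIndex switches both emitters between bright HDR gold at zero and blue otherwise, also changing spark alpha scale.
# An FN_Integrator value feeds that alpha through a lerp with identical endpoints, so it has no effect.
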NiagaraSystem /BRCosmetics/Weapons/FORT_Melee/Pickaxe_Dino_Hunter_Female/FX/NS_Dino_Hunter_Female_Impact
dependencies {
 usedDataInterfaces [&NiagaraDataInterfaceCurve, &NiagaraDataInterfaceVector2DCurve]
 usedDynamicInputs [
  &ClampFloat
  &SetBoolByFloatComparison
  &VectorLength
 ]
 usedModules [&SystemState, &FN_Integrator]
}
emitters {
 Spark_Impact {
  emitterUpdateScript [
   &EmitterState()
   &SpawnBurst_Instantaneous(Spawn Count=1, Use Spawn Probability=false)
  ]
  particleSpawnScript [&InitializeParticle(
    Color Mode=::NewEnumerator0
    Lifetime=0.125
    Sprite Size Mode=::NewEnumerator3
    Uniform Sprite Size=70
    UsePositionOffset=false
    Write Material Random=true
   )]
  particleUpdateScript [
   &ParticleState()
   &ScaleSpriteSize(Scale Factor=&Vector2DFromCurve(
     Vector2Curve=DataInterface() {propertyValues {
       XCurve {keys [<at 0 0.5>, {interpMode=RCIM_Cubic, leaveTangent=-2.8598, tangentMode=RCTM_Break, time=0.2, value=1}, <at 1 0>]}
       YCurve {keys [<at 0 0.5>, {interpMode=RCIM_Cubic, leaveTangent=-2.8598, tangentMode=RCTM_Break, time=0.2, value=1}, <at 1 0>]}
      }}
     ScaleCurve=<1 1>
    )) {module=ScaleSpriteSize001}
   &Color(Color=&MakeCustomLinearColorFromBool(Bool=HlslExpression() {hlslExpression="User.SmokeColorIndex == 0"}, True Color=<rgba 50 41.1997 10.578 1>, False Color=<rgba 0 10.7668 25 1>))
   &CameraOffset(Camera Offset Amount=5)
  ]
  rendererClasses [&NiagaraSpriteRendererProperties]
  renderers [{rendererClass=&NiagaraSpriteRendererProperties}]
 }
 Sparks {
  emitterUpdateScript [
   &EmitterState()
   &SpawnBurst_Instantaneous(Spawn Count=&RandomRangeInt(Minimum=55, Maximum=90), Use Spawn Probability=false)
  ]
  particleSpawnScript [
   &InitializeParticle(
    Lifetime=&RandomRangeFloat(Minimum=0.25, Maximum=0.62)
    Mass Mode=::NewEnumerator1
    Sprite Rotation Angle=&RandomRangeFloat(Minimum=0, Maximum=360)
    Sprite Rotation Mode=::NewEnumerator0
    Sprite Size=&Multiply_Vector2(A=<2 20>, B=&Vector2DFromFloat(Value=2))
    Sprite Size Mode=::NewEnumerator4
    UsePositionOffset=false
    Write Material Random=true
   )
   &AddVelocityInCone(
    Cone Angle=30
    Module.Use Velocity Falloff On Cone Axis=false
    Velocity Strength=&RandomRangeFloat(Minimum=100, Maximum=1500)
   )
  ]
  particleUpdateScript [
   &ParticleState()
   &Color(
    Color=&MakeCustomLinearColorFromBool(Bool=HlslExpression() {hlslExpression="User.SmokeColorIndex == 0"}, True Color=<rgba 50 41.1997 10.578 1>, False Color=<rgba 0 10.7668 25 1>)
    Scale Alpha=&FloatFromCurve(
     FloatCurve=DataInterface() {propertyValues {Curve {keys [
         {interpMode=RCIM_Cubic, leaveTangent=14.6157, tangentMode=RCTM_Break, time=0, value=0}
         {interpMode=RCIM_Cubic, leaveTangent=-2.2426, tangentMode=RCTM_Break, time=0.25, value=1.5}
         {interpMode=RCIM_Cubic, tangentMode=RCTM_Break, time=1, value=0}
        ]}}}
     Scale Curve=&Multiply_Float(
      A=&MakeCustomFloatFromBool(
       Bool=HlslExpression() {hlslExpression="User.SmokeColorIndex == 0.0"}
       True Float=&RandomRangeFloat(Minimum=0.2, Maximum=0.6)
       False Float=&RandomRangeFloat(Minimum=0.1, Maximum=0.3)
      )
      B=&Lerp_Float(A=1, B=1, Alpha=$System.FN_Integrator.Value)
     )
    )
   )
   &Drag(Drag=10, Use Linear Drag=true, Use Rotational Drag=false)
   &ScaleSpriteSize(Scale Factor=&Vector2DFromCurve(
     Vector2Curve=DataInterface() {propertyValues {
       XCurve {keys [{interpMode=RCIM_Cubic, leaveTangent=-2.8598, tangentMode=RCTM_Break, time=0, value=1}, <at 1 0>]}
       YCurve {keys [{interpMode=RCIM_Cubic, leaveTangent=-2.8598, tangentMode=RCTM_Break, time=0, value=1}, <at 1 0>]}
      }}
     ScaleCurve=<1 1>
    )) {module=ScaleSpriteSize001}
   &GravityForce(Gravity=<0 0 -1200>)
   &CurlNoiseForce(Noise Frequency=2500, Noise Strength=1500, Pan Noise=false)
   &SolveForcesAndVelocity(Clamp Velocity=false, Limit Acceleration=false)
  ]
  rendererClasses [&NiagaraSpriteRendererProperties]
  renderers [{rendererClass=&NiagaraSpriteRendererProperties}]
 }
}
summary {
 systemName=NS_Dino_Hunter_Female_Impact
 emitters [
  {emitterName=Sparks, rendererClasses [&NiagaraSpriteRendererProperties]}
  {emitterName=Spark_Impact, rendererClasses [&NiagaraSpriteRendererProperties]}
 ]
 userVariables [User.SmokeColorIndex {type=NiagaraFloat, defaultValue=0, description=""}]
}
