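# Physics asset for the Dino Hunter Female two-handed pickaxe, giving its dangling chain physics bodies.
# A long capsule on the handle bone anchors three small capsules on the dyn_chains bones, linked in series by constraints.
# Each joint allows wide swing on one axis and only a few degrees otherwise, with the free axis alternating down the chain like interlocking links.
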PhysicsAsset /BRCosmetics/Weapons/FORT_Melee/Pickaxe_Dino_Hunter_Female/Meshes/Dino_Hunter_Female_dyn
bodies {
 Melee_TwoHanded_Handle [{shapeType=Capsule, length=53.0387, radius=2.7934, center=<0.4361 -0.099 3.7548>, rotation=<rot 0.3805 -90.0014 0.087>}]
 dyn_chains_01 [{shapeType=Capsule, length=3.9543, radius=2.208, center=<2 0 0>, rotation=<rot 0 -90 -90>}]
 dyn_chains_02 [{shapeType=Capsule, length=3.9543, radius=2.208, center=<2 0 0>, rotation=<rot 0 -90 -90>}]
 dyn_chains_03 [{shapeType=Capsule, length=3.9543, radius=2.208, center=<3 0 0>, rotation=<rot 0 -90 -90>}]
}
constraints [
 {bone1Name=dyn_chains_01, bone2Name=Melee_TwoHanded_Handle, swing1LimitDegrees=80, swing2LimitDegrees=2, twistLimitDegrees=5}
 {bone1Name=dyn_chains_02, bone2Name=dyn_chains_01, swing1LimitDegrees=2, swing2LimitDegrees=75, twistLimitDegrees=2}
 {bone1Name=dyn_chains_03, bone2Name=dyn_chains_02, swing1LimitDegrees=45, swing2LimitDegrees=5, twistLimitDegrees=2}
]
